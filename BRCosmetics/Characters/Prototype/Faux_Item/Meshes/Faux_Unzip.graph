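# Animation Blueprint for the Faux cosmetic character skeleton that follows the attached parent mesh's pose and curves, with room for DefaultSlot montages.
# Each update caches the owning PlayerPawn_Athena's anim instance, whose CopyPose curve, inverted, blends between the raw copied pose and the slot output.
# After evaluation, a zero-length delay precedes writing the index_03_rSocket location into the SocketLocation material parameter, exposing that fingertip position to shaders.
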AnimBlueprint /BRCosmetics/Characters/Prototype/Faux_Item/Meshes/Faux_Unzip
target_skeleton=/BRCosmetics/Characters/Player/Male/Medium/Bodies/M_MED_Faux/Meshes/M_MED_Faux_Skeleton
graphs [
 AnimGraph {nodes [
   n1 = Copy Pose From Mesh() {bCopyCurves=true, bUseAttachedParent=true}
   n2 = Get AttachedAnimInstance() -> (AttachedAnimInstance:Anim Instance Object Reference)
   Output Pose(Result=Two Way Blend(
     A=Copy Pose From Mesh() {bCopyCurves=true, bUseAttachedParent=true}
     B=Slot 'DefaultSlot'(Source=@n1) {bAlwaysUpdateSourcePose=true}
     Alpha=GetCurveValue(self=@n2, CurveName=CopyPose)
    ) {blendNode {alphaBoolBlend {blendOption=Linear}, alphaScaleBiasClamp {bias=1, scale=-1}}}) {layerGroup=None}
  ]}
 EventGraph {nodes [
   Event BlueprintUpdateAnimation()
   n1 = GetOwningActor()
   > Is Valid(InputObject=@n1)
   > n2 = Cast To PlayerPawn_Athena(Object=@n1)
   > Set AttachedAnimInstance(AttachedAnimInstance:Anim Instance Object Reference=GetAnimInstance(self=Get Mesh(self=@n2))) -> (Output_Get:Anim Instance Object Reference)
   Event BlueprintPostEvaluateAnimation() {is=title K2Node_Event}
   n3 = GetOwningComponent()
   > Is Valid(InputObject=@n3)
   > Delay(Duration=0)
   > SetVectorParameterValueOnMaterials!(
    execute
    self:Mesh Component Object Reference=@n3
    ParameterName:Name=SocketLocation
    ParameterValue:Vector=GetSocketLocation!(self:Scene Component Object Reference=@n3, InSocketName:Name=index_03_rSocket) -> (ReturnValue:Vector)
   ) -> (then)
   TryGetPawnOwner()
  ]}
]
variables [
 AttachedAnimInstance {type=AnimInstance, default=null}
 Mesh {type=SkeletalMeshComponent}
]
